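# Sound cue for a looping snow cable ambience in the Figment S09 Dopey event sounds, judging by its name and path.
# A looping wave player feeds Snow_Complete_Loop through a Modulator at default settings, at reduced cue volume under the medium-priority events sound class.
# An attenuation override on the cue sets distance falloff, low-pass filtering over distance, wide stereo spread, distance-scaled reverb and interpolated focus settings.
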SoundCue /Figment_S09/Athena/Sounds/Events/Dopey/Snow/Snow_Cable_Amb_Cue
root=Modulator_0
nodes [
 {id=Modulator_0, children [WavePlayer_1]}
 {id=WavePlayer_1, properties {bLooping=true, soundWaveAssetPtr=&/Snow_Complete_Loop}}
]
properties {
 bOverrideAttenuation=true
 maxDistance=2524
 soundClassObject=/Figment_S09/Athena/Sounds/Events/SoundClasses/Fort_Events_Priority_Medium
 volumeMultiplier=0.4
 attenuationOverrides {
  bAttenuateWithLPF=true
  bEnableFocusInterpolation=true
  falloffDistance=1500
  focusAttackInterpSpeed=3
  focusAzimuth=10
  lPFFrequencyAtMax=2000
  lPFFrequencyAtMin=5000
  lPFRadiusMax=3000
  lPFRadiusMin=0
  nonFocusAzimuth=30
  nonFocusVolumeAttenuation=0.3
  occlusionInterpolationTime=0.2
  occlusionLowPassFilterFrequency=1000
  occlusionTraceChannel=ECC_GameTraceChannel11
  reverbDistanceMax=3000
  reverbDistanceMin=1000
  reverbWetLevelMax=10
  reverbWetLevelMin=0
  stereoSpread=1024
  attenuationShapeExtents {x=1024}
 }
}
