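# Looping ambient Sound Cue for a damaged cable, filed under the Dopey event's Snow sounds in Figment_S09.
# A single looping wave player feeds the output with boosted volume, routed to an extra-high-priority events sound class.
# Its own attenuation override adds distance falloff, low-pass filtering, distance-scaled reverb, stereo spread and interpolated listener focus.
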SoundCue /Figment_S09/Athena/Sounds/Events/Dopey/Snow/Snow_DamagedCable_Amb_Cue
root=WavePlayer_0
nodes [{id=WavePlayer_0, properties {bLooping=true, soundWaveAssetPtr=&/Snow_Cable_DamageLoop_01}}]
properties {
 bOverrideAttenuation=true
 maxDistance=2262
 soundClassObject=/Figment_S09/Athena/Sounds/Events/SoundClasses/Fort_Events_Priority_ExtHigh
 volumeMultiplier=2
 attenuationOverrides {
  bEnableFocusInterpolation=true
  bEnableListenerFocus=true
  falloffDistance=1750
  focusAzimuth=15
  lPFFrequencyAtMax=2000
  lPFRadiusMax=3000
  lPFRadiusMin=0
  nonFocusAzimuth=30
  reverbDistanceMax=3000
  reverbDistanceMin=1000
  reverbWetLevelMax=10
  reverbWetLevelMin=0
  stereoSpread=256
  attenuationShapeExtents {x=512}
 }
}
