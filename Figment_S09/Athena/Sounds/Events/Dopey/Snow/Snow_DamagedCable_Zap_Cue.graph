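# Sound cue for electrical zaps from a damaged cable in the Dopey Snow event, on an extra-high-priority events sound class.
# A random node picks among seven equally weighted zap waves, then a modulator varies pitch and volume slightly.
# Inline attenuation overrides add distance low-pass filtering, reverb send and listener focus, which attenuates off-focus sources.
# A concurrency override caps simultaneous instances, stopping the oldest and applying volume scaling.
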SoundCue /Figment_S09/Athena/Sounds/Events/Dopey/Snow/Snow_DamagedCable_Zap_Cue
root=Modulator_0
nodes [
 {id=Modulator_0, children [Random_1], properties {pitchMax=1.1, pitchMin=0.9, volumeMax=1.1, volumeMin=0.8}}
 {
  id=Random_1
  children [WavePlayer_5, WavePlayer_6, WavePlayer_7, WavePlayer_8, WavePlayer_9, WavePlayer_10, WavePlayer_11]
  properties {weights [1,1,1,1,1,1,1]}
 }
 {id=WavePlayer_10, properties {soundWaveAssetPtr=&/Snow_Cable_Zap_05}}
 {id=WavePlayer_11, properties {soundWaveAssetPtr=&/Snow_Cable_Zap_06}}
 {id=WavePlayer_5, properties {soundWaveAssetPtr=&/Snow_Cable_Zap_03}}
 {id=WavePlayer_6, properties {soundWaveAssetPtr=&/Snow_Cable_Zap_07}}
 {id=WavePlayer_7, properties {soundWaveAssetPtr=&/Snow_Cable_Zap_01}}
 {id=WavePlayer_8, properties {soundWaveAssetPtr=&/Snow_Cable_Zap_02}}
 {id=WavePlayer_9, properties {soundWaveAssetPtr=&/Snow_Cable_Zap_04}}
]
properties {
 bOverrideAttenuation=true
 bOverrideConcurrency=true
 maxDistance=4024
 soundClassObject=/Figment_S09/Athena/Sounds/Events/SoundClasses/Fort_Events_Priority_ExtHigh
 attenuationOverrides {
  bAttenuateWithLPF=true
  bEnableFocusInterpolation=true
  bEnableListenerFocus=true
  falloffDistance=3000
  lPFFrequencyAtMax=500
  lPFRadiusMax=2048
  lPFRadiusMin=0
  nonFocusAzimuth=45
  nonFocusVolumeAttenuation=0.3
  reverbDistanceMax=3000
  reverbDistanceMin=1000
  reverbWetLevelMax=5
  reverbWetLevelMin=0
  stereoSpread=512
  attenuationShapeExtents {x=1024}
 }
 concurrencyOverrides {maxCount=5, resolutionRule=StopOldest, volumeScale=0.3, volumeScaleAttackTime=1}
}
